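# Physics asset for the crouched gun holster on the Tomcat medium male player body, simulating the dangling holster against the hips.
# Bodies cover the pelvis, right thigh and two dynamic holster bones, using capsules plus a box for the holster.
# Angular-limited constraints chain the lower holster bone to the upper one, which hangs from the pelvis with a tight swing and twist range.
# The thigh is also constrained to the pelvis.
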PhysicsAsset /BRCosmetics/Characters/Player/Male/Medium/Bodies/M_MED_Tomcat/Meshes/M_MED_Tomcat_Gun_Crouch_Physics
bodies {
 dyn_gunHolster_01 [{shapeType=Capsule, length=2, radius=1, center=<3 0 0>}]
 dyn_gunHolster_03 [{
   shapeType=Box
   extentX=14.5
   extentY=6.5
   extentZ=30
   shapeName=dyn_gunHolster_02_box
   center=<5.5796 -2.0001 4.3839>
   rotation=<rot 79.9969 -180 -180>
  }]
 pelvis [
  {shapeType=Capsule, length=8, radius=10}
  {shapeType=Capsule, length=6.5, radius=2.75, shapeName=pelvis_box, center=<4.1841 -4.6802 13.3033>, rotation=<rot 68.8764 -61.6323 -66.6725>}
 ]
 thigh_r [{shapeType=Capsule, length=20, radius=6.5, center=<10 -1 -1>, rotation=<rot 90 0 0>}]
}
constraints [
 {bone1Name=thigh_r, bone2Name=pelvis, swing1LimitDegrees=45, swing2LimitDegrees=45, twistLimitDegrees=45}
 {bone1Name=dyn_gunHolster_01, bone2Name=pelvis, swing1LimitDegrees=1, swing2LimitDegrees=20, twistLimitDegrees=1}
 {bone1Name=dyn_gunHolster_03, bone2Name=dyn_gunHolster_01, swing1LimitDegrees=20, swing2LimitDegrees=20, twistLimitDegrees=10}
]
