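# Physics asset for the gun holster on the M_MED_Tomcat male player body, covering the dynamic holster bones plus pelvis and right thigh.
# A three-bone dyn_gunHolster chain of capsules is constrained from the pelvis, with tight, mostly single-axis swing at the root and limited swing and twist further down.
# Pelvis and right thigh capsules, joined by a wider constraint, provide body volumes beside the holster.
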PhysicsAsset /BRCosmetics/Characters/Player/Male/Medium/Bodies/M_MED_Tomcat/Meshes/M_MED_Tomcat_Gun_Physics3
bodies {
 dyn_gunHolster_01 [{shapeType=Capsule, length=2, radius=1, center=<3 0 0>}]
 dyn_gunHolster_02 [{shapeType=Capsule, length=15, radius=3.5, shapeName=dyn_gunHolster_02_capsule, center=<15 -2 2>, rotation=<rot 100 0 0>}]
 dyn_gunHolster_03 [{shapeType=Capsule, length=15, radius=3.5, shapeName=dyn_gunHolster_02_capsule, center=<11 -2 2>, rotation=<rot 100 0 0>}]
 pelvis [{shapeType=Capsule, length=8, radius=10}]
 thigh_r [{shapeType=Capsule, length=20, radius=6.5, center=<10 -1 -1>, rotation=<rot 90 0 0>}]
}
constraints [
 {bone1Name=thigh_r, bone2Name=pelvis, swing1LimitDegrees=45, swing2LimitDegrees=45, twistLimitDegrees=45}
 {bone1Name=dyn_gunHolster_01, bone2Name=pelvis, swing1LimitDegrees=1, swing2LimitDegrees=20, twistLimitDegrees=1}
 {bone1Name=dyn_gunHolster_02, bone2Name=dyn_gunHolster_01, swing1LimitDegrees=15, swing2LimitDegrees=10, twistLimitDegrees=10}
 {bone1Name=dyn_gunHolster_03, bone2Name=dyn_gunHolster_02, swing1LimitDegrees=5, swing2LimitDegrees=5, twistLimitDegrees=5}
]
