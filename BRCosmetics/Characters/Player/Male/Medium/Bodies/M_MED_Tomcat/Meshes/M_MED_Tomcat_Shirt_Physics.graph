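# Physics asset for the shirt of the male medium Tomcat body, providing simulation bodies for a dangling shirt flap.
# Two small capsules on the dyn_shirtFlap bones form a chain hanging from the pelvis, with tighter twist and second-swing limits than the thigh joint.
# Larger pelvis and left thigh capsules give the body volumes around the flap, with the thigh constrained to the pelvis.
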PhysicsAsset /BRCosmetics/Characters/Player/Male/Medium/Bodies/M_MED_Tomcat/Meshes/M_MED_Tomcat_Shirt_Physics
bodies {
 dyn_shirtFlap_01 [{shapeType=Capsule, length=8, radius=1, shapeName=dyn_shirtFlap_01_capsule, center=<5 0 0>}]
 dyn_shirtFlap_02 [{shapeType=Capsule, length=5, radius=1.5, shapeName=dyn_shirtFlap_02_capsule, center=<5 0 0>}]
 pelvis [{shapeType=Capsule, length=8, radius=10}]
 thigh_l [{shapeType=Capsule, length=20, radius=7, center=<-10 1 1>, rotation=<rot 90 0 0>}]
}
constraints [
 {bone1Name=thigh_l, bone2Name=pelvis, swing1LimitDegrees=45, swing2LimitDegrees=45, twistLimitDegrees=45}
 {bone1Name=dyn_shirtFlap_02, bone2Name=dyn_shirtFlap_01, swing1LimitDegrees=45, swing2LimitDegrees=20, twistLimitDegrees=5}
 {bone1Name=dyn_shirtFlap_01, bone2Name=pelvis, swing1LimitDegrees=45, swing2LimitDegrees=20, twistLimitDegrees=5}
]
