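# Post-process style AnimBlueprint for the Tomcat male body cosmetic, copying the attached parent's pose and adding secondary motion to the shirt flaps and gun holster.
# Bone constraints feed three pelvis-space RigidBody simulations (shirt, gun, crouch gun), then Transform and Look At nodes adjust flaps and holster.
# The EventGraph maps thigh-to-pelvis rotation into normalized deltas and a crouch flag, with no visible AnimGraph reads.
# The dyn_gunHolster_02 nodes are unconnected.
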AnimBlueprint /BRCosmetics/Characters/Player/Male/Medium/Bodies/M_MED_Tomcat/Meshes/M_MED_Tomcat_AnimBP
target_skeleton=&/M_MED_Tomcat_Skeleton
graphs [
 AnimGraph {nodes [
   "Constraint - dyn_gunHolster_02 (PP)"!(ComponentPose, Alpha=1, ConstraintWeights_0=0.7, ConstraintWeights_1:float=0.3) -> (Pose) {
    alphaBoolBlend {blendOption=Linear}
    boneToModify {boneName=dyn_gunHolster_02}
    constraintSetup [{targetBone {boneName=thigh_r}}, {targetBone {boneName=pelvis}}]
    constraintWeights [1,1]
   }
   Look At - Bone: dyn_gunHolster_02(Alpha=1) {
    alphaBoolBlend {blendOption=Linear}
    boneToModify {boneName=dyn_gunHolster_02}
    lookAtLocation=<-2.5 -2>
    lookAtTarget {boneReference {boneName=thigh_r}}
   }
   "Transform (Modify) Bone - Bone: dyn_gunHolster_02"(Alpha=1, Translation="5,-5,0", Rotation="0,0,0", Scale="1,1,1") {
    translationMode=BMM_Additive
    translationSpace=BCS_BoneSpace
    alphaBoolBlend {blendOption=Linear}
    boneToModify {boneName=dyn_gunHolster_02}
    translation=<5 -5>
   }
   n1 = "Constraint - dyn_shirtFlap_01 (PP)"!(
    ComponentPose=Local To Component(LocalPose=Copy Pose From Mesh() {bUseAttachedParent=true})
    Alpha=1
    ConstraintWeights_0=0.5
    ConstraintWeights_1:float=0.5
   ) -> (Pose) {
    alphaBoolBlend {blendOption=Linear}
    boneToModify {boneName=dyn_shirtFlap_01}
    constraintSetup [{targetBone {boneName=pelvis}}, {targetBone {boneName=spine_01}}]
    constraintWeights [1,1]
   }
   n2 = "Constraint - dyn_gunHolster_01 (PP)"!(ComponentPose=@n1, Alpha=1, ConstraintWeights_0=0.7, ConstraintWeights_1:float=0.3) -> (Pose) {
    alphaBoolBlend {blendOption=Linear}
    boneToModify {boneName=dyn_gunHolster_01}
    constraintSetup [{targetBone {boneName=pelvis}}, {targetBone {boneName=spine_01}}]
    constraintWeights [1,1]
   }
   n3 = RigidBody(ComponentPose=@n2) {
    bClampLinearTranslationLimitToRefPose=true
    bOverrideWorldGravity=true
    overridePhysicsAsset=&/M_MED_Tomcat_Shirt_Physics
    simulationSpace=BaseBoneSpace
    alphaBoolBlend {blendOption=Linear}
    alphaScaleBiasClamp {bInterpResult=true}
    baseBoneRef {boneName=pelvis}
    componentAppliedLinearAccClamp=<3000 3000 3000>
    componentLinearAccScale=<0.4 0.3 0.3>
    componentLinearVelScale=<0.4 0.3 0.3>
    overrideWorldGravity {z=-980}
   }
   n4 = RigidBody(ComponentPose=@n3) {
    bClampLinearTranslationLimitToRefPose=true
    bOverrideWorldGravity=true
    overridePhysicsAsset=&/M_MED_Tomcat_Gun_Physics3
    simulationSpace=BaseBoneSpace
    alphaBoolBlend {blendOption=Linear}
    alphaScaleBiasClamp {bInterpResult=true}
    baseBoneRef {boneName=pelvis}
    componentAppliedLinearAccClamp=<3000 3000 3000>
    componentLinearAccScale=<0.4 0.3 0.3>
    componentLinearVelScale=<0.4 0.3 0.3>
    overrideWorldGravity {z=-980}
   }
   n5 = RigidBody(ComponentPose=@n4) {
    bClampLinearTranslationLimitToRefPose=true
    bOverrideWorldGravity=true
    overridePhysicsAsset=&/M_MED_Tomcat_Gun_Crouch_Physics
    simulationSpace=BaseBoneSpace
    alphaBoolBlend {blendOption=Linear}
    alphaScaleBiasClamp {bInterpResult=true}
    baseBoneRef {boneName=pelvis}
    componentAppliedLinearAccClamp=<3000 3000 3000>
    componentLinearAccScale=<0.4 0.3 0.3>
    componentLinearVelScale=<0.4 0.3 0.3>
    overrideWorldGravity {z=-980}
   }
   n6 = "Transform (Modify) Bone - Bone: dyn_gunHolster_01"(ComponentPose=@n5, Alpha=1, Translation="-0.5,0,2", Rotation="60,30,90", Scale="1,1,1") {
    translationMode=BMM_Additive
    translationSpace=BCS_BoneSpace
    alphaBoolBlend {blendOption=Linear}
    boneToModify {boneName=dyn_gunHolster_01}
    rotation=<rot 60 30 90>
    translation {x=-0.5, z=2}
   }
   n7 = Look At - Bone: dyn_shirtFlap_01(ComponentPose=@n6, Alpha=1) {
    alphaBoolBlend {blendOption=Linear}
    boneToModify {boneName=dyn_shirtFlap_01}
    lookAtLocation {x=0}
    lookAtTarget {boneReference {boneName=thigh_twist_01_l}}
    lookAt_Axis {axis=<1 0.5 -0.3>}
   }
   n8 = "Transform (Modify) Bone - Bone: dyn_shirtFlap_01"(ComponentPose=@n7, Alpha=1, Translation="0,-5.5,-4", Rotation="0,0,0", Scale="1,1,1") {
    rotationMode=BMM_Additive
    rotationSpace=BCS_BoneSpace
    translationMode=BMM_Additive
    translationSpace=BCS_BoneSpace
    alphaBoolBlend {blendOption=Linear}
    boneToModify {boneName=dyn_shirtFlap_01}
    translation {y=-5.5, z=-4}
   }
   n9 = Look At - Bone: dyn_shirtFlap_02(ComponentPose=@n8, Alpha=1) {
    alphaBoolBlend {blendOption=Linear}
    boneToModify {boneName=dyn_shirtFlap_02}
    lookAtLocation {x=0}
    lookAtTarget {boneReference {boneName=thigh_twist_01_l}}
    lookAt_Axis {axis=<1 1.5 -0.3>}
   }
   n10 = "Transform (Modify) Bone - Bone: dyn_shirtFlap_02"(ComponentPose=@n9, Alpha=1, Translation="-3,1,0", Rotation="0,0,0", Scale="1,1,1") {
    rotationMode=BMM_Additive
    rotationSpace=BCS_BoneSpace
    translationMode=BMM_Additive
    translationSpace=BCS_BoneSpace
    alphaBoolBlend {blendOption=Linear}
    boneToModify {boneName=dyn_shirtFlap_02}
    translation=<-3 1>
   }
   Output Pose(Result=Component To Local(ComponentPose=@n10)) {layerGroup=None}
  ]}
 EventGraph {nodes [
   Event BlueprintUpdateAnimation()
   > n1 = Sequence()
   > n2 = Branch(Condition=Get bIsCrouching())
   n3 = GetOwnerSkeletalMeshComponent()
   n4 = BreakTransform(InTransform=GetDeltaTransformFromRefPose(self=@n3, BoneName=pelvis, BaseName=thigh_l))
   > n5 = Set thigh_lDelta(thigh_lDelta:double=MapRangeClamped(Value=BreakRotator(InRot=@n4.Rotation).Yaw, InRangeA=40, InRangeB=115)) -> (Output_Get:double)
   > Set invThigh_LDelta(invThigh_LDelta:double=float - float(B=@n5) {is=title K2Node_CallFunction}) -> (Output_Get:double)
   > Set notCrouching(notCrouching:bool=false) -> (Output_Get:bool)
   n6 = Set thigh_lDelta(execute=@n2.else, thigh_lDelta=0)
   > Set invThigh_LDelta(invThigh_LDelta=float - float(B=@n6))
   > Set notCrouching(notCrouching=true)
   n7 = BreakTransform(InTransform=GetDeltaTransformFromRefPose(self=@n3, BoneName=pelvis, BaseName=thigh_l))
   Set thigh_lDeltaX(execute=@n1.then_1, thigh_lDeltaX:double=MapRangeClamped(Value=BreakRotator(InRot=@n7.Rotation).Pitch, InRangeA=-5, InRangeB=-37.9)) -> (Output_Get:double)
   TryGetPawnOwner()
  ]}
]
variables [
 invThigh_LDelta {type=number}
 notCrouching {type=boolean}
 thigh_lDelta {type=number}
 thigh_lDeltaX {type=number}
]
