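# Physics asset for the Tomcat medium male character's face accessory mesh, defining bodies for dynamic hair bones plus head and neck.
# Seven hair capsules are constrained directly to the head, with tight angular limits on front and side pieces and much looser swing on the three back pieces.
# The head combines a capsule and a box, and links through two neck capsules by constraints with free swing and twist.
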PhysicsAsset /BRCosmetics/Characters/Player/Male/Medium/Bodies/M_MED_Tomcat/Meshes/Parts/M_MED_Tomcat_FaceAcc_Physics
bodies {
 dyn_hair_bk_l_01_new [{shapeType=Capsule, length=3, radius=1.5, shapeName=dyn_hair_bk_l_01_new_capsule, center=<3.2 0 0>, rotation=<rot 90 0 0>}]
 dyn_hair_bk_new [{shapeType=Capsule, length=3, radius=1.5, shapeName=dyn_hair_bk_new_capsule, center=<3.2 0 0>, rotation=<rot 90 0 0>}]
 dyn_hair_bk_r_01_new [{shapeType=Capsule, length=3, radius=1.5, shapeName=dyn_hair_bk_r_01_new_capsule, center=<3.2 0 0>, rotation=<rot 90 0 0>}]
 dyn_hair_fr_l [{shapeType=Capsule, length=2, radius=1, shapeName=dyn_hair_fr_l_capsule, center=<0.5 0 0>, rotation=<rot 90 0 0>}]
 dyn_hair_fr_r [{shapeType=Capsule, length=2, radius=1, shapeName=dyn_hair_fr_r_capsule, center=<0.5 0 0>, rotation=<rot 90 0 0>}]
 dyn_hair_l [{shapeType=Capsule, length=2, radius=1, shapeName=dyn_hair_l_capsule, center=<0.5 0 0>, rotation=<rot 90 0 0>}]
 dyn_hair_r [{shapeType=Capsule, length=2, radius=1, shapeName=dyn_hair_r_capsule, center=<0.5 0 0>, rotation=<rot 90 0 0>}]
 head [
  {shapeType=Capsule, length=5.25, radius=5.25, center=<3 3 0>, rotation=<rot 90 0 0>}
  {shapeType=Box, extentX=7.5, extentY=7.5, extentZ=11, shapeName=head_box, center=<1.8236 2.4698 0>}
 ]
 neck_01 [{shapeType=Capsule, length=3.25, radius=3.25, shapeName=neck_01_capsule, rotation=<rot 90 0 0>}]
 neck_02 [{shapeType=Capsule, length=3.25, radius=3.25, shapeName=neck_02_capsule, rotation=<rot 90 0 0>}]
}
constraints [
 {bone1Name=dyn_hair_fr_l, bone2Name=head, swing1LimitDegrees=5, swing2LimitDegrees=10, twistLimitDegrees=2}
 {bone1Name=dyn_hair_fr_r, bone2Name=head, swing1LimitDegrees=5, swing2LimitDegrees=10, twistLimitDegrees=2}
 {bone1Name=dyn_hair_r, bone2Name=head, swing1LimitDegrees=5, swing2LimitDegrees=5, twistLimitDegrees=2}
 {bone1Name=dyn_hair_l, bone2Name=head, swing1LimitDegrees=5, swing2LimitDegrees=5, twistLimitDegrees=2}
 {bone1Name=dyn_hair_bk_r_01_new, bone2Name=head, swing1LimitDegrees=45, swing2LimitDegrees=45, twistLimitDegrees=5}
 {bone1Name=dyn_hair_bk_new, bone2Name=head, swing1LimitDegrees=45, swing2LimitDegrees=45, twistLimitDegrees=5}
 {bone1Name=dyn_hair_bk_l_01_new, bone2Name=head, swing1LimitDegrees=45, swing2LimitDegrees=45, twistLimitDegrees=5}
 {
  bone1Name=head
  bone2Name=neck_02
  swing1LimitDegrees=45
  swing1Motion=Free
  swing2LimitDegrees=45
  swing2Motion=Free
  twistLimitDegrees=45
  twistMotion=Free
 }
 {
  bone1Name=neck_02
  bone2Name=neck_01
  swing1LimitDegrees=45
  swing1Motion=Free
  swing2LimitDegrees=45
  swing2Motion=Free
  twistLimitDegrees=45
  twistMotion=Free
 }
]
